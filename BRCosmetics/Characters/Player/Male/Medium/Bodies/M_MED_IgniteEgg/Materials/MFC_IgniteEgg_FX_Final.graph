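# Material function for the IgniteEgg player skin, adding hologram, glitch and hit effects to incoming material attributes.
# Blueprint-fed timestamps drive hologram transitions, downed-state (DBNO) glitch bursts and hit responses with blocky centers and expanding sphere-mask waves.
# Time-stepped noise textures form flickering glitch masks for view-space vertex offsets and an extra GlitchUV output.
# Tinted scanlines in stabilized screen space feed emissive, with debug switches for hit and hologram timing.
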MaterialFunction /BRCosmetics/Characters/Player/Male/Medium/Bodies/M_MED_IgniteEgg/Materials/MFC_IgniteEgg_FX_Final
graph {nodes [
  Constant()
  Constant() {r=1}
  Custom!(time, speed1, speed2, speed3) {
   code="float ts1 = (sin(time*speed1)+1)/2;\r\nfloat ts2 = (sin(time*speed2)+1)/2;\r\nfloat ts3 = (sin(time*speed3)+1)/2;\r\nreturn lerp(ts1, ts2, ts3);"
   showCode=true
   inputs [
    {inputName=time, input {expression=null}}
    {inputName=speed1, input {expression=null}}
    {inputName=speed2, input {expression=null}}
    {inputName=speed3, input {expression=null}}
   ]
  }
  Frac(None=Multiply(A=Time(), B=ScalarParameter() {defaultValue=-0.5, group=IgniteEgg, parameterName=Debug_HologramSpeed}))
  n1 = GetMaterialAttributes(None=FunctionInput() {inputName=Material, inputType=FunctionInput_MaterialAttributes}) {attributeGetTypes [#1, #2, #3]}
  FunctionOutput(None=SetMaterialAttributes!(MaterialAttributes=@n1.MaterialAttributes, OpacityMask, EmissiveColor=NamedRerouteUsage(), WorldPositionOffset=NamedRerouteUsage()) {attributeSetTypes [#1, #2, #3]})
  n2 = TextureCoordinate()
  n3 = NamedRerouteDeclaration(None=Multiply(A=NamedRerouteUsage(), B=ScalarParameter() {defaultValue=0.1, group=IgniteEgg, parameterName=IE_Ambient_GlitchIntensity})) {name=AmbientGlitchMask, nodeColor=<rgba 0.5234 0 1 1>}
  FunctionOutput(None=NamedRerouteDeclaration(None=Add(A=@n2, B=@n3)) {name=GlitchUV, nodeColor=<rgba 0 1 0.7988 1>}) {outputName=GlitchUV}
  n4 = Frac(None=Multiply(A=Time(), B=ScalarParameter() {defaultValue=0.75, group=IgniteEgg, parameterName=Debug_HitSpeed}))
  n5 = Subtract(A=Time(), B=ScalarParameter() {group=IgniteEgg, parameterName=IE_GameTimeFromBP_Hit})
  n6 = Divide(A=@n5, B=ScalarParameter() {defaultValue=0.9, group=IgniteEgg, parameterName=IE_HitDuration})
  n7 = NamedRerouteDeclaration(None=Saturate(None=StaticSwitchParameter(True=@n4, False=@n6) {group=IgniteEgg, parameterName=DEBUG_Hit})) {name=HitProgress, nodeColor=<rgba 0.5 0 1 1>}
  n8 = Power(Base=OneMinus(None=@n7)) {constExponent=4}
  n9 = LocalPosition()
  n10 = Constant() {r=128}
  n11 = Constant() {r=25}
  n12 = Floor(None=Multiply(A=Divide(A=@n9.XYZ, B=@n10) {constB=128}, B=@n11))
  n13 = VectorParameter() {group=IgniteEgg, parameterName=HitVector, defaultValue=<rgba 0 0 100 1>}
  n14 = SphereMask(A=Multiply(A=Divide(A=@n12, B=@n11), B=@n10), B=@n13.RGB, Radius=LinearInterpolate(Alpha=@n7) {constA=10, constB=35}) {attenuationRadius=30, hardnessPercent=1}
  n15 = Subtract(A=Multiply(A=Multiply(A=@n8, B=@n14)) {constB=5}, B=Multiply(A=NamedRerouteUsage()) {constB=0.5})
  n16 = NamedRerouteUsage()
  n17 = SmoothStep(Value=Saturate(None=Subtract(A=@n15, B=@n16))) {constMax=0.5}
  Multiply(A=@n17, B=@n8)
  n18 = NamedRerouteUsage()
  n19 = MaterialFunctionCall!(
   ScanLine_Tiling_1
   Scanline_PannerSpeed_1=ScalarParameter() {defaultValue=-0.2, group=IgniteEgg, parameterName=ScanlinePanSpeed}
   ScanLine_Tiling=Constant() {r=100}
   Scanline_PannerSpeed
   Scanline_Rotation
   UVs=MaterialFunctionCall!(ObjectPivotLocation, ObjectPositionOffset) {materialFunction=/Game/Characters/Player/Common/MaterialFunctions/Superlevel/ImprovedStabilizedScreenSpaceUVs}.StabilizedUVs
  ) {desc=ScanLine, materialFunction=/Game/Global/Materials/MaterialGym/Materials/HolographicTexture/MF_HolographicTexture_ScanLines}
  n20 = LinearInterpolate(
   B=Multiply(A=VectorParameter() {group=IgniteEgg, parameterName=IE_HologramTint, defaultValue=<rgba 1.3877 1.0104 2 1>}.RGB, B=Add(A=@n19.Result))
   Alpha=Add(A=NamedRerouteUsage(), B=NamedRerouteUsage())
  ) {constA=1}
  n21 = Multiply(A=Multiply(A=@n19.Result, B=NamedRerouteUsage()), B=NamedRerouteUsage())
  n22 = Constant() {r=-1}
  NamedRerouteDeclaration(None=Add(A=Multiply(A=@n18, B=@n20), B=Desaturation(None=@n21, Fraction=@n22))) {name=Emissive_OUT, nodeColor=<rgba 0.8047 0 1 1>}
  n23 = MaterialFunctionCall!(Position, UVRadialScale) {materialFunction=/Game/Weapons/WeaponSkins/MaterialFunction/ObjectStabilizedScreenSpaceUVs}
  n24 = TextureSample(UVs=Panner(Coordinate=NamedRerouteDeclaration(None=@n23.StabilizedUVs) {name=ScreenSpaceUV, nodeColor=<rgba 0.4062 0 1 1>}) {bFractionalPart=true, speedX=-1.5, speedY=0.987}) {
   samplerSource=SSM_Wrap_WorldGroupSettings
   samplerType=SAMPLERTYPE_LinearColor
   texture=/Game/Global/Textures/Default/Noise/T_Atlas_DigitalNoise_01
  }
  NamedRerouteDeclaration(None=@n24.RGB) {name=RandomGlitchColor, nodeColor=<rgba 1 0.1641 0 1>}
  NamedRerouteDeclaration(None=@n1.WorldPositionOffset) {name=WPO_IN, nodeColor=<rgba 0.4297 1 0 1>}
  n25 = Add(A=Time(), B=Multiply(A=ScalarParameter() {group=IgniteEgg, parameterName=IE_RandomX}) {constB=50})
  n26 = Sine(None=Multiply(A=@n25, B=Constant() {r=0.27}))
  n27 = Sine(None=Multiply(A=@n25, B=Constant() {r=0.15}))
  n28 = Constant() {r=0.1}
  n29 = ConstantBiasScale(None=Sine(None=Multiply(A=@n25, B=@n28)))
  n30 = SmoothStep(
   Min=LinearInterpolate(
    A=ScalarParameter() {defaultValue=0.9, group=IgniteEgg, parameterName=GlitchFrequency_Ambient}
    B=ScalarParameter() {defaultValue=0.36, group=IgniteEgg, parameterName=GlitchFrequency_DBNO}
    Alpha=NamedRerouteUsage()
   ) {constA=0.9}
   Value=LinearInterpolate(A=ConstantBiasScale(None=@n26), B=ConstantBiasScale(None=@n27), Alpha=@n29)
  ) {constMin=0.9}
  n31 = ScalarParameter() {defaultValue=3, group=IgniteEgg, parameterName=GlitchSpeed2}
  n32 = Floor(None=Multiply(A=Time(), B=@n31) {constB=6})
  n33 = Add(A=Multiply(A=NamedRerouteUsage()) {constB=2}, B=ScalarParameter() {group=IgniteEgg, parameterName=IE_RandomY})
  n34 = TextureSample(UVs=Add(A=Divide(A=@n32, B=@n31) {constB=3}, B=@n33)) {
   samplerSource=SSM_Wrap_WorldGroupSettings
   samplerType=SAMPLERTYPE_LinearColor
   texture=/Game/Environments/Asteria/Sets/NeonCity/Techart/Holograms/GlitchMap_M
  }
  n35 = Step(Y=LinearInterpolate(Alpha=@n29) {constA=0.9, constB=0.99}, X=@n34.R) {constY=0.99}
  n36 = Multiply(A=LinearInterpolate(A=Multiply(A=@n30, B=@n35) {desc=Flicker}, B=@n34.R, Alpha=NamedRerouteUsage()), B=NamedRerouteUsage()) {desc=Samller local area}
  NamedRerouteDeclaration(None=@n36) {name=Debug, nodeColor=<rgba 1 0.6797 0 1>}
  n37 = Multiply(A=NamedRerouteUsage(), B=@n36)
  NamedRerouteDeclaration(None=Add(A=NamedRerouteUsage(), B=@n37)) {name=GlitchMask0-1, nodeColor=<rgba 1 0.9609 0 1>}
  n38 = NamedRerouteUsage()
  n39 = ScalarParameter() {defaultValue=15, parameterName=GlitchSpeed}
  n40 = Multiply(A=Time(), B=@n39) {constB=6}
  n41 = Frac(None=Divide(A=Floor(None=@n40), B=@n39) {constB=3})
  n42 = TextureSample(UVs=Add(A=Multiply(A=@n38) {constB=7}, B=ConstantBiasScale(None=@n41) {bias=-0.5, scale=2})) {
   samplerSource=SSM_Wrap_WorldGroupSettings
   samplerType=SAMPLERTYPE_LinearColor
   texture=&GlitchMap_M
  }
  NamedRerouteDeclaration(None=LinearInterpolate(A=@n42.G, B=NamedRerouteUsage(), Alpha=NamedRerouteUsage())) {name=GlitchPattern, nodeColor=<rgba 1 0.1406 0 1>}
  n43 = LinearInterpolate(Alpha=NamedRerouteUsage()) {constA=0.8, constB=0.3}
  n44 = NamedRerouteUsage()
  n45 = TextureSample(UVs=@n33) {samplerType=SAMPLERTYPE_Alpha, texture=/BRCosmetics/Characters/Player/Male/Medium/Bodies/M_MED_IgniteEgg/Prototype/Textures/T_IgniteEgg_TA2}
  n46 = Time()
  n47 = LinearInterpolate(A=LinearInterpolate(Alpha=NamedRerouteUsage()) {constA=0.3}, Alpha=NamedRerouteUsage()) {constB=2}
  n48 = Add(A=@n45.R, B=Frac(None=Multiply(A=@n46, B=@n47) {constB=0.5}) {desc=Random flicker speed})
  NamedRerouteDeclaration(None=SmoothStep(Min=LinearInterpolate(A=@n43, Alpha=@n44) {constB=0}, Value=Frac(None=@n48)) {constMin=0.9}) {name=RandomSquare, nodeColor=<rgba 1 0 0.6328 1>}
  n49 = MaterialFunctionCall(Float2=Clamp(None=ConstantBiasScale(None=@n37) {bias=-0.5, scale=2}) {minDefault=-1}) {materialFunction=&BreakOutFloat2Components}
  n50 = Constant()
  n51 = Multiply(
   A=Transform(None=MaterialFunctionCall(X=@n49.R, Y=@n50, Z=@n50) {materialFunction=&MakeFloat3}.Result) {transformSourceType=TRANSFORMSOURCE_View}
   B=LinearInterpolate(A=ScalarParameter() {defaultValue=5, group=IgniteEgg, parameterName=Glitch_WPOIntensity}, Alpha=NamedRerouteUsage()) {constB=10}
  )
  NamedRerouteDeclaration(None=Add(A=Multiply(A=@n37, B=@n51), B=NamedRerouteUsage())) {name=WPO_OUT, nodeColor=<rgba 0.2891 0 1 1>}
  n52 = NamedRerouteUsage()
  n53 = Constant2Vector() {g=7, r=1}
  n54 = ScalarParameter() {defaultValue=15, group=IgniteEgg, parameterName=GlitchSpeed}
  n55 = Floor(None=Multiply(A=Time(), B=@n54) {constB=6})
  n56 = TextureSample(UVs=Add(A=Multiply(A=@n52, B=@n53) {constB=5}, B=Divide(A=@n55, B=@n54) {constB=3})) {
   samplerSource=SSM_Wrap_WorldGroupSettings
   samplerType=SAMPLERTYPE_LinearColor
   texture=&GlitchMap_M
  }
  NamedRerouteDeclaration(None=@n56.B) {name=AmbientGlitchPattern, nodeColor=<rgba 0.7578 1 0 1>}
  NamedRerouteDeclaration(None=@n1.EmissiveColor) {name=Emissive_In, nodeColor=<rgba 0.0312 0 1 1>}
  n57 = Subtract(A=Time(), B=ScalarParameter() {group=IgniteEgg, parameterName=IE_GameTimeFromBP})
  n58 = ScalarParameter() {defaultValue=1, group=IgniteEgg, parameterName=IE_HologramFX, sliderMax=1}
  n59 = LinearInterpolate(
   A=ScalarParameter() {defaultValue=1.2, group=IgniteEgg, parameterName=IE_HologramDuration_In}
   B=ScalarParameter() {defaultValue=1.75, group=IgniteEgg, parameterName=IE_HologramDuration_Out}
   Alpha=NamedRerouteDeclaration(None=Saturate(None=@n58)) {name=HologramToggle, nodeColor=<rgba 0 1 0.8594 1>}
  )
  n60 = StaticSwitchParameter(True=ScalarParameter() {group=IgniteEgg, parameterName=Debug_Time, sliderMax=1}, False=Saturate(None=Divide(A=@n57, B=@n59))) {group=IgniteEgg, parameterName=DEBUG_Hologram}
  n61 = LinearInterpolate(A=@n60, B=OneMinus(None=@n60), Alpha=NamedRerouteUsage())
  n62 = ScalarParameter() {defaultValue=0.4, desc=0.5 blue, group=IgniteEgg, parameterName=IE_DBNO_HologramIntensity}
  n63 = NamedRerouteUsage()
  n64 = LinearInterpolate(B=@n62, Alpha=@n63) {constA=1, desc=into DBNO}
  n65 = LinearInterpolate(B=@n62, Alpha=@n63) {constB=0, desc=out off DBNO}
  n66 = NamedRerouteUsage()
  n67 = ScalarParameter() {group=IgniteEgg, parameterName=IE_IsDBNO}
  NamedRerouteDeclaration(None=LinearInterpolate(
    A=NamedRerouteDeclaration(None=@n61) {name=HologramProgress, nodeColor=<rgba 1 0 0.75 1>}
    B=LinearInterpolate(A=@n64, B=@n65, Alpha=@n66)
    Alpha=NamedRerouteDeclaration(None=@n67) {name=IsDBNO, nodeColor=<rgba 1 0.8438 0 1>}
   )) {name=HologramOpacity, nodeColor=<rgba 0 1 0.6016 1>}
  NamedRerouteDeclaration(None=@n1.OpacityMask) {name=OpacityMask_In, nodeColor=<rgba 0.2891 1 0 1>}
  NamedRerouteDeclaration() {name=OpacityMask_OUT, nodeColor=<rgba 1 0 0.6094 1>}
  n68 = LocalPosition() {localOrigin=InstancePreSkinning}
  NamedRerouteDeclaration(None=ShaderStageSwitch(PixelShader=VertexInterpolator(VS=@n68.XYZ).PS, VertexShader=@n68.XYZ)) {name=PS-LocalPos, nodeColor=<rgba 0.1719 1 0 1>}
  n69 = Multiply(A=@n17, B=ScalarParameter() {defaultValue=2, group=IgniteEgg, parameterName=HitCenterIntencity})
  n70 = Saturate(None=Divide(A=@n5, B=ScalarParameter() {defaultValue=1.1, group=IgniteEgg, parameterName=IE_HitDuration_2}))
  n71 = Power(Base=OneMinus(None=@n70)) {constExponent=3.8}
  n72 = SphereMask(A=LocalPosition().XYZ, B=@n13.RGB, Radius=LinearInterpolate(Alpha=@n70) {constB=90}) {attenuationRadius=30, hardnessPercent=1}
  n73 = Multiply(A=@n72, B=@n71)
  n74 = SphereMask(A=@n72, B=Constant() {r=0.5}) {attenuationRadius=0.5, hardnessPercent=0}
  n75 = LinearInterpolate(Alpha=@n71) {constA=1, constB=0.3}
  n76 = Saturate(None=Sine(None=Multiply(A=@n74, B=@n75) {constB=2}))
  n77 = Multiply(
   A=Multiply(A=Multiply(A=@n73, B=@n76), B=ScalarParameter() {defaultValue=10, group=IgniteEgg, parameterName=IE_WavesIntensity})
   B=NamedRerouteUsage()
  ) {desc=Radial expanding wave}
  NamedRerouteDeclaration(None=Multiply(A=Add(A=@n69, B=@n77), B=ScalarParameter() {defaultValue=1, group=IgniteEgg, parameterName=HitIntensity})) {name=HitMask, nodeColor=<rgba 1 0 0.4922 1>}
  n78 = Saturate(None=Divide(A=@n57, B=ScalarParameter() {defaultValue=0.85, group=IgniteEgg, parameterName=IE_DBNOglitchBurstDuration}))
  n79 = NamedRerouteUsage()
  NamedRerouteDeclaration(None=Multiply(A=OneMinus(None=@n78), B=OneMinus(None=@n79))) {name=GlitchBurstProgress, nodeColor=<rgba 1 0 0.0469 1>}
  NamedRerouteUsage()
  NamedRerouteUsage()
  NamedRerouteUsage()
  NamedRerouteUsage()
 ]}
