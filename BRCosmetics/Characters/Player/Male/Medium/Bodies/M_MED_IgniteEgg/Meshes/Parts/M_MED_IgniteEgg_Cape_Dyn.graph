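# Physics asset for the IgniteEgg male medium body's cape, with capsules for the cape bones, torso, shoulders, upper arms and legs.
# Four cape chains hang from clavicle_cape and backpack_root bones, with tight twist and swing1 limits and wider swing2.
# Angularly free cross-constraints tie the chains together side to side partway down, linking the cape like a sheet.
# Body joints are angularly free, and oversized calf capsules bulk out the lower legs.
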PhysicsAsset /BRCosmetics/Characters/Player/Male/Medium/Bodies/M_MED_IgniteEgg/Meshes/Parts/M_MED_IgniteEgg_Cape_Dyn
bodies {
 backpack_root [{shapeType=Capsule, length=1, radius=1}]
 calf_l [{shapeType=Capsule, length=11.7963, radius=40.5787, center=<-26.2275 20.2058 0>, rotation=<rot 90 -1.4887 0>}]
 calf_r [{shapeType=Capsule, length=11.7963, radius=40.5787, center=<26.2275 -20.2058 0>, rotation=<rot 90 178.5113 0>}]
 cape_01_l [{shapeType=Capsule, length=1, radius=1}]
 cape_01_r [{shapeType=Capsule, length=1, radius=1}]
 cape_02_l [{shapeType=Capsule, length=12, radius=9, center=<9 0 10>, rotation=<rot 90 0 0>}]
 cape_02_r [{shapeType=Capsule, length=12, radius=9, center=<9 0 10>, rotation=<rot 90 0 0>}]
 cape_03_l [{shapeType=Capsule, length=12, radius=12, center=<12 0 10>, rotation=<rot 90 0 0>}]
 cape_03_r [{shapeType=Capsule, length=12, radius=12, center=<12 0 10>, rotation=<rot 90 0 0>}]
 cape_04_l [{shapeType=Capsule, length=12, radius=12.723, center=<9 0 13.502>, rotation=<rot 90 0 0>}]
 cape_04_r [{shapeType=Capsule, length=12, radius=12.723, center=<9 0 13.502>, rotation=<rot 90 0 0>}]
 cape_05_l [{shapeType=Capsule, length=12, radius=12.723, center=<9 0 13.502>, rotation=<rot 90 0 0>}]
 cape_05_r [{shapeType=Capsule, length=12, radius=12.723, center=<9 0 13.502>, rotation=<rot 90 0 0>}]
 cape_06_l [{shapeType=Capsule, length=12, radius=12.723, center=<9 0 13.502>, rotation=<rot 90 0 0>}]
 cape_06_r [{shapeType=Capsule, length=12, radius=12.723, center=<9 0 13.502>, rotation=<rot 90 0 0>}]
 cape_07_l [{shapeType=Capsule, length=12, radius=12.723, center=<9 0 13.502>, rotation=<rot 90 0 0>}]
 cape_07_r [{shapeType=Capsule, length=12, radius=12.723, center=<9 0 13.502>, rotation=<rot 90 0 0>}]
 cape_mid_01_l [{shapeType=Capsule, length=12, radius=9, center=<9 0 10>, rotation=<rot 90 0 0>}]
 cape_mid_01_l_MSTR [{shapeType=Capsule, length=1, radius=1}]
 cape_mid_01_r [{shapeType=Capsule, length=12, radius=9, center=<9 0 10>, rotation=<rot 90 0 0>}]
 cape_mid_01_r_MSTR [{shapeType=Capsule, length=1, radius=1}]
 cape_mid_02_l [{shapeType=Capsule, length=12, radius=12, center=<12 0 10>, rotation=<rot 90 0 0>}]
 cape_mid_02_r [{shapeType=Capsule, length=12, radius=12, center=<12 0 10>, rotation=<rot 90 0 0>}]
 cape_mid_03_l [{shapeType=Capsule, length=12, radius=12.723, center=<9 0 13.502>, rotation=<rot 90 0 0>}]
 cape_mid_03_r [{shapeType=Capsule, length=12, radius=12.723, center=<9 0 13.502>, rotation=<rot 90 0 0>}]
 cape_mid_04_l [{shapeType=Capsule, length=12, radius=12.723, center=<9 0 13.502>, rotation=<rot 90 0 0>}]
 cape_mid_04_r [{shapeType=Capsule, length=12, radius=12.723, center=<9 0 13.502>, rotation=<rot 90 0 0>}]
 cape_mid_05_l [{shapeType=Capsule, length=12, radius=12.723, center=<9 0 13.502>, rotation=<rot 90 0 0>}]
 cape_mid_05_r [{shapeType=Capsule, length=12, radius=12.723, center=<9 0 13.502>, rotation=<rot 90 0 0>}]
 cape_mid_06_l [{shapeType=Capsule, length=12, radius=12.723, center=<9 0 13.502>, rotation=<rot 90 0 0>}]
 cape_mid_06_r [{shapeType=Capsule, length=12, radius=12.723, center=<9 0 13.502>, rotation=<rot 90 0 0>}]
 clavicle_cape_l [{shapeType=Capsule, length=1, radius=1}]
 clavicle_cape_r [{shapeType=Capsule, length=1, radius=1}]
 clavicle_l [{shapeType=Capsule, length=12.9688, radius=9.6562, center=<12.3638 -1.6165 0>, rotation=<rot 90 0 -6.5611>}]
 clavicle_r [{shapeType=Capsule, length=12.9688, radius=9.6562, center=<-12.3638 1.6165 0>, rotation=<rot 90 0 173.4389>}]
 pelvis [{shapeType=Capsule, length=6.0938, radius=16.25, center=<-0.9978 -1.6863 0>}]
 spine_01 [{shapeType=Capsule, length=4.0312, radius=13.4062, center=<3.5949 0.0155 0>}]
 spine_02 [{shapeType=Capsule, length=3.8125, radius=13.625, center=<3.5949 0.0155 0>}]
 spine_03 [{shapeType=Capsule, length=39.6562, radius=18, center=<2.5875 -0.252 0>}]
 spine_04 [{shapeType=Capsule, length=31.4062, radius=17.4375, center=<2.8832 -0.1618 0>}]
 spine_05 [{shapeType=Capsule, length=32.875, radius=15.375, center=<-0.0029 0.1172 0>}]
 thigh_l [{shapeType=Capsule, length=28.0938, radius=10.7188, center=<-19.7778 -0.1788 0>, rotation=<rot 90 18.9704 16.3433>}]
 thigh_r [{shapeType=Capsule, length=28.0938, radius=10.7188, center=<19.7778 0.1788 0>, rotation=<rot 90 -10.6197 166.7532>}]
 upperarm_l [{shapeType=Capsule, length=20.625, radius=6.9375, center=<11.9672 -1.662 0>, rotation=<rot 90 0 -6.5611>}]
 upperarm_r [{shapeType=Capsule, length=20.625, radius=6.9375, center=<-11.9672 1.662 0>, rotation=<rot 90 0 173.4389>}]
}
constraints [
 {
  bone1Name=spine_01
  bone2Name=pelvis
  swing1LimitDegrees=45
  swing1Motion=Free
  swing2LimitDegrees=45
  swing2Motion=Free
  twistLimitDegrees=45
  twistMotion=Free
 }
 {
  bone1Name=spine_02
  bone2Name=spine_01
  swing1LimitDegrees=45
  swing1Motion=Free
  swing2LimitDegrees=45
  swing2Motion=Free
  twistLimitDegrees=45
  twistMotion=Free
 }
 {
  bone1Name=spine_03
  bone2Name=spine_02
  swing1LimitDegrees=45
  swing1Motion=Free
  swing2LimitDegrees=45
  swing2Motion=Free
  twistLimitDegrees=45
  twistMotion=Free
 }
 {
  bone1Name=spine_04
  bone2Name=spine_03
  swing1LimitDegrees=45
  swing1Motion=Free
  swing2LimitDegrees=45
  swing2Motion=Free
  twistLimitDegrees=45
  twistMotion=Free
 }
 {
  bone1Name=spine_05
  bone2Name=spine_04
  swing1LimitDegrees=45
  swing1Motion=Free
  swing2LimitDegrees=45
  swing2Motion=Free
  twistLimitDegrees=45
  twistMotion=Free
 }
 {
  bone1Name=backpack_root
  bone2Name=spine_05
  swing1LimitDegrees=45
  swing1Motion=Free
  swing2LimitDegrees=45
  swing2Motion=Free
  twistLimitDegrees=45
  twistMotion=Free
 }
 {
  bone1Name=cape_mid_01_l_MSTR
  bone2Name=backpack_root
  swing1LimitDegrees=45
  swing1Motion=Free
  swing2LimitDegrees=45
  swing2Motion=Free
  twistLimitDegrees=45
  twistMotion=Free
 }
 {bone1Name=cape_mid_01_l, bone2Name=cape_mid_01_l_MSTR, swing1LimitDegrees=5, swing2LimitDegrees=45, twistLimitDegrees=5}
 {
  bone1Name=cape_mid_01_r_MSTR
  bone2Name=backpack_root
  swing1LimitDegrees=45
  swing1Motion=Free
  swing2LimitDegrees=45
  swing2Motion=Free
  twistLimitDegrees=45
  twistMotion=Free
 }
 {bone1Name=cape_mid_01_r, bone2Name=cape_mid_01_r_MSTR, swing1LimitDegrees=5, swing2LimitDegrees=45, twistLimitDegrees=5}
 {bone1Name=cape_01_l, bone2Name=clavicle_cape_l, swing1LimitDegrees=45, swing2LimitDegrees=45, twistLimitDegrees=5}
 {bone1Name=cape_02_l, bone2Name=cape_01_l, swing1LimitDegrees=5, swing2LimitDegrees=45, twistLimitDegrees=5}
 {bone1Name=cape_03_l, bone2Name=cape_02_l, swing1LimitDegrees=5, swing2LimitDegrees=45, twistLimitDegrees=5}
 {bone1Name=cape_04_l, bone2Name=cape_03_l, swing1LimitDegrees=10, swing2LimitDegrees=45, twistLimitDegrees=1}
 {bone1Name=cape_05_l, bone2Name=cape_04_l, swing1LimitDegrees=10, swing2LimitDegrees=45, twistLimitDegrees=1}
 {bone1Name=cape_06_l, bone2Name=cape_05_l, swing1LimitDegrees=10, swing2LimitDegrees=45, twistLimitDegrees=1}
 {bone1Name=cape_07_l, bone2Name=cape_06_l, swing1LimitDegrees=10, swing2LimitDegrees=45, twistLimitDegrees=1}
 {bone1Name=cape_01_r, bone2Name=clavicle_cape_r, swing1LimitDegrees=45, swing2LimitDegrees=45, twistLimitDegrees=5}
 {bone1Name=cape_02_r, bone2Name=cape_01_r, swing1LimitDegrees=5, swing2LimitDegrees=45, twistLimitDegrees=5}
 {bone1Name=cape_03_r, bone2Name=cape_02_r, swing1LimitDegrees=5, swing2LimitDegrees=45, twistLimitDegrees=5}
 {bone1Name=cape_04_r, bone2Name=cape_03_r, swing1LimitDegrees=10, swing2LimitDegrees=45, twistLimitDegrees=1}
 {bone1Name=cape_05_r, bone2Name=cape_04_r, swing1LimitDegrees=10, swing2LimitDegrees=45, twistLimitDegrees=1}
 {bone1Name=cape_06_r, bone2Name=cape_05_r, swing1LimitDegrees=10, swing2LimitDegrees=45, twistLimitDegrees=1}
 {bone1Name=cape_07_r, bone2Name=cape_06_r, swing1LimitDegrees=10, swing2LimitDegrees=45, twistLimitDegrees=1}
 {
  bone1Name=thigh_l
  bone2Name=pelvis
  swing1LimitDegrees=45
  swing1Motion=Free
  swing2LimitDegrees=45
  swing2Motion=Free
  twistLimitDegrees=45
  twistMotion=Free
 }
 {
  bone1Name=calf_l
  bone2Name=thigh_l
  swing1LimitDegrees=45
  swing1Motion=Free
  swing2LimitDegrees=45
  swing2Motion=Free
  twistLimitDegrees=45
  twistMotion=Free
 }
 {
  bone1Name=upperarm_l
  bone2Name=clavicle_l
  swing1LimitDegrees=45
  swing1Motion=Free
  swing2LimitDegrees=45
  swing2Motion=Free
  twistLimitDegrees=45
  twistMotion=Free
 }
 {
  bone1Name=upperarm_r
  bone2Name=clavicle_r
  swing1LimitDegrees=45
  swing1Motion=Free
  swing2LimitDegrees=45
  swing2Motion=Free
  twistLimitDegrees=45
  twistMotion=Free
 }
 {
  bone1Name=calf_r
  bone2Name=thigh_r
  swing1LimitDegrees=45
  swing1Motion=Free
  swing2LimitDegrees=45
  swing2Motion=Free
  twistLimitDegrees=45
  twistMotion=Free
 }
 {
  bone1Name=thigh_r
  bone2Name=pelvis
  swing1LimitDegrees=45
  swing1Motion=Free
  swing2LimitDegrees=45
  swing2Motion=Free
  twistLimitDegrees=45
  twistMotion=Free
 }
 {
  bone1Name=clavicle_l
  bone2Name=spine_05
  swing1LimitDegrees=45
  swing1Motion=Free
  swing2LimitDegrees=45
  swing2Motion=Free
  twistLimitDegrees=45
  twistMotion=Free
 }
 {
  bone1Name=clavicle_r
  bone2Name=spine_05
  swing1LimitDegrees=45
  swing1Motion=Free
  swing2LimitDegrees=45
  swing2Motion=Free
  twistLimitDegrees=45
  twistMotion=Free
 }
 {bone1Name=cape_mid_02_l, bone2Name=cape_mid_01_l, swing1LimitDegrees=5, swing2LimitDegrees=45, twistLimitDegrees=5}
 {bone1Name=cape_mid_03_l, bone2Name=cape_mid_02_l, swing1LimitDegrees=10, swing2LimitDegrees=45, twistLimitDegrees=1}
 {bone1Name=cape_mid_04_l, bone2Name=cape_mid_03_l, swing1LimitDegrees=10, swing2LimitDegrees=45, twistLimitDegrees=1}
 {bone1Name=cape_mid_05_l, bone2Name=cape_mid_04_l, swing1LimitDegrees=10, swing2LimitDegrees=45, twistLimitDegrees=1}
 {bone1Name=cape_mid_06_l, bone2Name=cape_mid_05_l, swing1LimitDegrees=10, swing2LimitDegrees=45, twistLimitDegrees=1}
 {bone1Name=cape_mid_02_r, bone2Name=cape_mid_01_r, swing1LimitDegrees=5, swing2LimitDegrees=45, twistLimitDegrees=5}
 {bone1Name=cape_mid_03_r, bone2Name=cape_mid_02_r, swing1LimitDegrees=10, swing2LimitDegrees=45, twistLimitDegrees=1}
 {bone1Name=cape_mid_04_r, bone2Name=cape_mid_03_r, swing1LimitDegrees=10, swing2LimitDegrees=45, twistLimitDegrees=1}
 {bone1Name=cape_mid_05_r, bone2Name=cape_mid_04_r, swing1LimitDegrees=10, swing2LimitDegrees=45, twistLimitDegrees=1}
 {bone1Name=cape_mid_06_r, bone2Name=cape_mid_05_r, swing1LimitDegrees=10, swing2LimitDegrees=45, twistLimitDegrees=1}
 {
  bone1Name=cape_mid_04_l
  bone2Name=cape_mid_04_r
  swing1LimitDegrees=45
  swing1Motion=Free
  swing2LimitDegrees=45
  swing2Motion=Free
  twistLimitDegrees=45
  twistMotion=Free
 }
 {
  bone1Name=cape_05_l
  bone2Name=cape_mid_04_l
  swing1LimitDegrees=45
  swing1Motion=Free
  swing2LimitDegrees=45
  swing2Motion=Free
  twistLimitDegrees=45
  twistMotion=Free
 }
 {
  bone1Name=cape_mid_04_r
  bone2Name=cape_05_r
  swing1LimitDegrees=45
  swing1Motion=Free
  swing2LimitDegrees=45
  swing2Motion=Free
  twistLimitDegrees=45
  twistMotion=Free
 }
]
